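# Light function material that casts animated dappled light and shadow, projected on world XY.
# A rotating spiral texture multiplies a panning Voronoi cloud texture, and a squared, opposite-panning Voronoi layer at its own tiling modulates the result.
# The pattern is inverted, shaped by power and brightness parameters, then lifted by a shadow-lightening floor.
# A static switch drives motion from scaled Time or a fixed time parameter.
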Material /BlastBerryMapContent/Game/TimeOfDay/LightFunctions/M_Rufus_LF_Dappled
graph {
 properties {materialDomain=MD_LightFunction, pixelDepthOffsetMode=PDOM_Legacy, shadingModel=MSM_Unlit, usedShadingModels=Unlit}
 nodes [
  n1 = WorldPosition()
  n2 = Divide(A=@n1.XY, B=ScalarParameter() {defaultValue=5, parameterName=ShadowTile1_Divisor})
  n3 = StaticSwitchParameter(
   True=Multiply(A=ScalarParameter() {defaultValue=1, parameterName=Speed_Panner_Multi}, B=Time())
   False=ScalarParameter() {defaultValue=1, parameterName=Time}
  ) {defaultValue=true, parameterName=Animate?}
  n4 = Multiply(
   A=TextureSample(UVs=Rotator(Coordinate=@n2, Time=@n3) {speed=0.003}) {samplerType=SAMPLERTYPE_Grayscale, texture=/Game/Environments/World/LightFunctions/T_spiralLF}.R
   B=TextureSample(UVs=Panner(Coordinate=@n2, Time=@n3) {speedY=0.01}) {samplerType=SAMPLERTYPE_LinearGrayscale, texture=/Game/Packages/Fortress_Sky/SkyDome/Master/T_SkyDome_Cloud_Voroni}.R
  )
  n5 = Constant() {r=-0.1497}
  n6 = Constant()
  n7 = ScalarParameter() {defaultValue=2, parameterName=ShadowTile2_Divisor}
  n8 = TextureSample(UVs=Panner(Coordinate=Divide(A=@n1.XY, B=@n7), Time=@n3) {speedY=-0.01}) {samplerType=SAMPLERTYPE_LinearGrayscale, texture=&T_SkyDome_Cloud_Voroni}
  n9 = Multiply(A=@n8.R, B=@n8.R)
  n10 = Subtract(A=@n4, B=Multiply(A=OneMinus(None=@n4), B=LinearInterpolate(A=@n5, B=@n6, Alpha=@n9) {constB=0.1}))
  n11 = Power(Base=OneMinus(None=Saturate(None=@n10)), Exp=ScalarParameter() {defaultValue=7, parameterName=FinalPower})
  n12 = ScalarParameter() {defaultValue=4, parameterName=FinalBrightness}
  LinearInterpolate(A=ScalarParameter() {defaultValue=0.35, parameterName=LightenShadow}, Alpha=Saturate(None=Multiply(A=@n11, B=@n12)))
 ]
}
